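; Blueprint for the Dragon's Breath break-action shotgun (Athena variant), adding its impact and shell-eject effects.
; On impact, it spawns a Dragon's Breath emitter at the spawned particle component's transform unless the surface is SurfaceType8, in which case it lowers that component's DebrisAmount.
; The muzzle flash event calls the parent, then after a short delay restarts the shell-eject particle component by deactivating and reactivating it.
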(event FortWeapon|EventOnPlayImpactFX
  (params (HitResult :type Hit%20Result%20Structure%20%28by%20ref%29) (ImpactPhysicalSurface :type EPhysicalSurface%20Enum :default "SurfaceType_Default") (SpawnedPSC :type FXSystem%20Component%20Object%20Reference))
  (if (not (Utilities|Enum|Equal(Enum) ImpactPhysicalSurface "SurfaceType8"))
    (Effects|SpawnEmitteratLocation "/Game/Weapons/FORT_Shotguns/Effects/Impacts/P_DragonsBreath_Shotgun_Break_Impact_Athena.P_DragonsBreath_Shotgun_Break_Impact_Athena" (Transformation|GetWorldLocation SpawnedPSC) (Transformation|GetWorldRotation SpawnedPSC))
    (Effects|Components|ParticleSystem|SetFloatParameter SpawnedPSC "DebrisAmount" 0.1)))

(event EventMuzzleFlashFX
  (params (Persistent%20Fire :type Boolean))
  (bind _p_shell_eject_dragonsbreath (Variables|Default|GetP_Shell_Eject_DragonsBreath))
  (|CallParentFunction Persistent%20Fire)
  (Utilities|FlowControl|Delay 0.8)
  (Components|Activation|Deactivate _p_shell_eject_dragonsbreath)
  (Components|Activation|Activate _p_shell_eject_dragonsbreath true))

(fn ConstructionScript
  (|CallParentFunction))
