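; Athena charge shotgun weapon Blueprint that adds charge-related audio cues and effects to its parent weapon class.
; While charging, a looping timer shows a gunshot sound indicator for hostile and neutral affiliations, clearing itself once charging ends.
; A local max-charge flag triggers an attached steam-escape sound on unequip.
; Fire FX activate a fully-charged Niagara effect on full charge, marked unused in comments, and impact FX set debris amount.
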(event FortWeapon|EventOnPlayImpactFX
  (params (HitResult :type Hit%20Result%20Structure%20%28by%20ref%29) (ImpactPhysicalSurface :type EPhysicalSurface%20Enum :default "SurfaceType_Default") (SpawnedPSC :type FXSystem%20Component%20Object%20Reference))
  ;[0.1 0.1 0.1] Debris Amount
  (|CallParentFunction HitResult ImpactPhysicalSurface SpawnedPSC)
  (Effects|Components|ParticleSystem|SetFloatParameter SpawnedPSC "DebrisAmount" 0.2))

(event EventOnPlayWeaponFireFX
  (params (bPersistentFire :type Boolean) (bSecondaryFire :type Boolean))
  ;[1 1 1] M uzzle Flashes [Overcharged state currently not in use ]
  (|CallParentFunction bPersistentFire bSecondaryFire)
  (switch Utilities|FlowControl|Switch|SwitchonEFortWeaponChargeStateForFireFX (FortWeapon|GetWeaponChargeStateforFireFX)
    (:Full
      ;;[1 0.06 0.15 0.9 11] Unused
      (Components|Activation|Activate (Variables|Default|GetNS_FullyCharged) true))))

(event EventOnStartCharge
  ;[1 1 1] Setting a max charge variable and cleaning it up for weapon swap steam escaping audio
  (Variables|Default|SetMaxChargeForSwap false)
  (Variables|Default|SetAudioVisualizerEnabled true)
  ;[1 1 1] Audio Visualizer Timer and Cleanup
  (bind _returnvalue (Utilities|Time|SetTimerbyEvent (delegate Custom|ChargeAudioVisualizer) 0.75 true))
  (Variables|Default|SetAudioVisualizerTimer _returnvalue))

(event EventOnLocalClientReachedMaxCharge
  ;[1 1 1] Setting a max charge variable and cleaning it up for weapon swap steam escaping audio
  (Variables|Default|SetMaxChargeForSwap true))

(event Custom|ChargeAudioVisualizer
  ;[1 1 1] Audio Visualizer Timer and Cleanup
  (if (Variables|Default|GetAudioVisualizerEnabled)
    (Fortnite|ShowSoundIndicatorLocalforTeam :Location (Transformation|GetActorLocation) :MaxAudibleDistance 3e+03 :IndicatorType "Gunshot" :Team (Team|GetActorTeam self) :Affiliations (Utilities|Array|MakeArray "Hostile" "Neutral") :Tint "(R=0.000000,G=0.000000,B=0.000000,A=0.000000)" :AttenuationMaxRangeMultiplier 1.0 :AttenuationIntensityMultiplier 1.0)
    (else
      (Utilities|Time|ClearandInvalidateTimerbyHandle (Variables|Default|GetAudioVisualizerTimer)))))

(event EventOnEndCharge
  ;[1 1 1] Setting a max charge variable and cleaning it up for weapon swap steam escaping audio
  (Variables|Default|SetMaxChargeForSwap false)
  (Variables|Default|SetAudioVisualizerEnabled false))

(event FortWeapon|EventOnLocalUnEquip
  ;[1 1 1] Setting a max charge variable and cleaning it up for weapon swap steam escaping audio
  (|CallParentFunction)
  (if (Variables|Default|GetMaxChargeForSwap)
    (Audio|SpawnSoundAttached (Variables|Default|GetMaxChargeSwapAudio) (Variables|Visual|GetWeaponMesh))))

(fn ConstructionScript
  (|CallParentFunction))
