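; Retro sci-fi shotgun Blueprint that adds a selectable neon color to its effects and darkens the weapon's glow when the magazine is effectively empty.
; On BeginPlay it picks color values from option arrays, then passes them to muzzle-flash and impact particles.
; Ammo changes update mesh pulse and emissive parameters only when the empty state flips.
; Material updates wait until cosmetic alterations (wraps) initialize, which also sets the initial state.
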(event Custom|NeonColor%20Selection
  (bind _color_selection (Variables|Neon|GetColor%20Selection))
  (bind _returnvalue (Math|Conversions|ToInteger(Byte) _color_selection))
  (Variables|Neon|SetNeon_Color (Utilities|Array|Get(aref) (Variables|Neon|GetNeon_Color_Options) _returnvalue))
  (Variables|Neon|SetNeon_Color_Selection (Utilities|Array|Get(acopy) (Variables|Neon|GetNeon_Color_Selection_Options) _returnvalue)))

(event EventBeginPlay
  (|CallParentFunction)
  (CallFunction|NeonColorSelection))

(event EventMuzzleFlashFX
  (params (Persistent%20Fire :type Boolean))
  (bind _spawned_muzzle_system (Variables|Effects|Muzzle|GetSpawned%20Muzzle%20System))
  (|CallParentFunction Persistent%20Fire)
  (Effects|Components|ParticleSystem|SetVectorParameter _spawned_muzzle_system "NeonColor" (Math|Conversions|ToVector(LinearColor) (Variables|Neon|GetNeon_Color)))
  (Effects|Components|ParticleSystem|SetFloatParameter _spawned_muzzle_system "ColorSelect" (Variables|Neon|GetNeon_Color_Selection))
  (Effects|Components|ParticleSystem|SetVectorParameter _spawned_muzzle_system "NeonValue" (Math|Conversions|ToVector(Float) (Variables|Neon|GetNeon_Color_Selection))))

(event FortWeapon|EventOnPlayImpactFX
  (params (HitResult :type Hit%20Result%20Structure%20%28by%20ref%29) (ImpactPhysicalSurface :type EPhysicalSurface%20Enum :default "SurfaceType_Default") (SpawnedPSC :type FXSystem%20Component%20Object%20Reference))
  (|CallParentFunction HitResult ImpactPhysicalSurface SpawnedPSC)
  (Effects|Components|ParticleSystem|SetVectorParameter SpawnedPSC "NeonColor" (Math|Conversions|ToVector(LinearColor) (Variables|Neon|GetNeon_Color)))
  (Effects|Components|ParticleSystem|SetFloatParameter SpawnedPSC "ColorSelect" (Variables|Neon|GetNeon_Color_Selection)))

(event EventOnInitCosmeticAlterations
  (params (CosmeticMod :type Fort%20Cosmetic%20Modification%20Structure))
  ;[1 1 1 0.4] Set flag that wrap has been applied and initialize material parameters.
  (|CallParentFunction CosmeticMod)
  (Variables|Default|SetbWrapHasBeenApplied true)
  (bind _ammo_in_clip_is_very_low (CallFunction|DetemineWhetherAmmoinClipIsVeryLow))
  (bind _output_get (Variables|Default|SetbAmmoInClipIsVeryLow _ammo_in_clip_is_very_low))
  (Variables|Default|SetbAmmoInClipWasVeryLow _output_get)
  (CallFunction|SetMaterialParametersBasedonAmmo) ; Initialize Material Parameters
)

(event EventOnAmmoCountChanged
  ;[1 1 1 0.4] Update whether or not the Ammo in the gun is low and update the material if appropriate
  (|CallParentFunction)
  (bind _ammo_in_clip_is_very_low (CallFunction|DetemineWhetherAmmoinClipIsVeryLow))
  (if (Math|Boolean|NotEqual(Boolean) _ammo_in_clip_is_very_low (Variables|Default|GetbAmmoInClipIsVeryLow))
    (Variables|Default|SetbAmmoInClipWasVeryLow (Variables|Default|GetbAmmoInClipIsVeryLow)) ; Store the current value as the previous value
    (Variables|Default|SetbAmmoInClipIsVeryLow _ammo_in_clip_is_very_low) ; And set the new value
    (CallFunction|SetMaterialParametersBasedonAmmo) ; Update the material parameters
  ))

(event Custom|SetMaterialParametersBasedOnAmmo
  ;[1 1 1 0.4] Update Material Parameters based on whether or not ammo in the clip is low.
  ; Can only be done if wraps (if any exist) have already applied.  (We will also call this event to initialize as soon as any wrap (or no wrap) has been applied.)
  (if (Variables|Default|GetbWrapHasBeenApplied)
    (if (Variables|Default|GetbAmmoInClipIsVeryLow)
      (Rendering|Material|SetScalarParameterValueonMaterials (Variables|Visual|GetWeaponMesh) "Pulse_Intensity" 0.0)
      (Rendering|Material|SetScalarParameterValueonMaterials (Variables|Visual|GetWeaponMesh) "EmissiveMultiplier" 0.0)
      (else
        (Rendering|Material|SetScalarParameterValueonMaterials (Variables|Visual|GetWeaponMesh) "Pulse_Intensity" (Variables|Neon|GetPulse_Intensity))
        (Rendering|Material|SetScalarParameterValueonMaterials (Variables|Visual|GetWeaponMesh) "EmissiveMultiplier" (Variables|Neon|GetBase%20Emissive%20Multiplier))))))

(fn DetemineWhetherAmmoInClipIsVeryLow
  (return (<= (/ (Math|Conversions|ToFloat(Integer) (FortWeapon|GetMagazineAmmoCount)) (Math|Conversions|ToFloat(Integer) (FortWeapon|GetBulletsPerClip))) 0.0001)))

(fn ConstructionScript
  (|CallParentFunction))
