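# M_TorusRefraction_01 is a translucent unlit particle material for shotgun effects, built for index-of-refraction distortion and usable on mesh particles and Niagara sprites, ribbons and meshes.
# A Fresnel mask blends refraction from neutral toward a parameter value, scaled by a dynamic "Refraction Alpha" input.
# Particle color feeds an emissive adjustment function, and its alpha is multiplied by the Fresnel mask.
# A normal texture samples tiled UVs offset by fractional time times X and Y speed parameters.
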Material /Game/Weapons/FORT_Shotguns/Effects/Materials/M_TorusRefraction_01
graph {
 properties {
  bUsedWithMeshParticles=true
  bUsedWithNiagaraMeshParticles=true
  bUsedWithNiagaraRibbons=true
  bUsedWithNiagaraSprites=true
  blendMode=BLEND_Translucent
  pixelDepthOffsetMode=PDOM_Legacy
  refractionMethod=RM_IndexOfRefraction
  shadingModel=MSM_Unlit
  usedShadingModels=Unlit
 }
 nodes [
  Add()
  n1 = MaterialFunctionCall(
   Brightness=ScalarParameter() {defaultValue=1, parameterName=Fresnel_OuterMask-Brightness}
   Exponent=ScalarParameter() {defaultValue=2, parameterName=Fresnel_OuterMask-Exponent}
  ) {materialFunction=&MF_Fresnel_FX}
  LinearInterpolate(
   B=LinearInterpolate(B=ScalarParameter() {defaultValue=1.2, parameterName=Refraction}, Alpha=@n1.Fresnel) {constA=1}
   Alpha=DynamicParameter() {paramNames [Refraction Alpha, RandomUV-Y, Param3, Param4]}.Refraction Alpha
  ) {constA=1}
  n2 = ParticleColor()
  MaterialFunctionCall(Emissive=@n2.RGB) {materialFunction=&MF_FX_EmissiveAdjustments}
  Multiply(A=@n2.A, B=@n1.Fresnel)
  Panner()
  n3 = TextureCoordinate()
  n4 = Multiply(A=ComponentMask(None=@n3) {r=true}, B=Constant() {r=1})
  n5 = Frac(None=Time())
  n6 = Multiply(A=@n5, B=ScalarParameter() {parameterName=PanX- Speed}) {constB=-1}
  n7 = Multiply(A=ComponentMask(None=@n3) {g=true}, B=Constant() {r=2})
  n8 = Multiply(A=@n5, B=ScalarParameter() {defaultValue=-1, parameterName=PanY- Speed})
  TextureSample(UVs=AppendVector(A=Add(A=@n4, B=@n6), B=Add(A=@n7, B=@n8))) {samplerType=SAMPLERTYPE_Normal, texture=/Game/Global/Textures/Default/Tileable_Patterns/1Normal}
 ]
}
