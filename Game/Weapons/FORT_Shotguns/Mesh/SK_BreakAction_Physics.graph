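# Physics asset for the break-action shotgun skeletal mesh, providing simple collision and physics bodies for the weapon.
# The BreakAction and Handle bones each carry a single box body.
# One constraint joins the two bodies with swing and twist all free, so the listed angular limit angles do not restrict rotation.
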PhysicsAsset /Game/Weapons/FORT_Shotguns/Mesh/SK_BreakAction_Physics
bodies {
 BreakAction [{shapeType=Box, extentX=44.5237, extentY=12.042, extentZ=17.9579, center=<0.3094 13.5545 5.6342>, rotation=<rot 0 90.0001 0>}]
 Handle [{shapeType=Box, extentX=17.5137, extentY=27.0824, extentZ=23.5069, center=<0.3912 1.3541 0.901>}]
}
constraints [{
  bone1Name=BreakAction
  bone2Name=Handle
  swing1LimitDegrees=45
  swing1Motion=Free
  swing2LimitDegrees=45
  swing2Motion=Free
  twistLimitDegrees=45
  twistMotion=Free
 }]
