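# Physics asset for the Crusader shotgun skeletal mesh, in the Figment variant, in the shotgun weapons folder.
# It holds a single body on the Handle bone, shaped as one elongated box offset forward along X to enclose the weapon.
# It has no constraints, so the whole gun acts as one rigid piece for collision and physics.
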PhysicsAsset /Game/Weapons/FORT_Shotguns/Mesh/SK_Crusader_Physics_Figment
bodies {Handle [{shapeType=Box, extentX=95.4246, extentY=8.108, extentZ=21.5323, center=<19.5107 -0.0307 0.5508>}]}
constraints []
