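# Physics asset for the Hammer Shotgun weapon skeletal mesh, giving it simple collision and physics representation.
# It has a single body on the Handle bone, built from one offset box shape that is longest along X.
# It defines no constraints, so the mesh has no jointed or articulated physics parts.
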PhysicsAsset /Game/Weapons/FORT_Shotguns/Mesh/SK_HammerShotGun_Physics2
bodies {Handle [{shapeType=Box, extentX=82.4607, extentY=11.7096, extentZ=40.8918, center=<12.4032 0.8026 4.2464>}]}
constraints []
