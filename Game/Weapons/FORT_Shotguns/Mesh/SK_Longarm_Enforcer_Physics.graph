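# Physics asset for the Enforcer longarm shotgun skeletal mesh.
# A single box body on the Handle bone approximates the whole weapon for collision and physics simulation.
# It has no constraints, so the gun behaves as one rigid piece.
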PhysicsAsset /Game/Weapons/FORT_Shotguns/Mesh/SK_Longarm_Enforcer_Physics
bodies {Handle [{shapeType=Box, extentX=81.5345, extentY=13.1577, extentZ=28.5649, center=<16.6926 -1.3942 5.4583>}]}
constraints []
