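# Physics asset for the Post Apocalyptic Shotgun skeletal mesh among the Fortnite shotgun weapons.
# It holds a single body on the Handle bone, shaped as one elongated box offset along the bone's X axis to approximate the weapon as a whole.
# It defines no constraints, so the gun collides and simulates as one rigid piece with no separately moving parts.
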PhysicsAsset /Game/Weapons/FORT_Shotguns/Mesh/SK_Post_Apocalyptic_Shotgun_Physics
bodies {Handle [{shapeType=Box, extentX=99.6409, extentY=7.2581, extentZ=27.1004, center=<22.8247 -0.0133 3.9669>}]}
constraints []
