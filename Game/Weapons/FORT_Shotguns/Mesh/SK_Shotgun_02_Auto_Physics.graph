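# Physics asset for the SK_Shotgun_02_Auto skeletal mesh, an automatic shotgun weapon, giving it a simplified collision shape.
# It holds a single body on the Handle bone, made of one box offset forward and up from the bone origin.
# There are no constraints, so the weapon simulates as one rigid piece with no articulated parts.
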PhysicsAsset /Game/Weapons/FORT_Shotguns/Mesh/SK_Shotgun_02_Auto_Physics
bodies {Handle [{shapeType=Box, extentX=74.9706, extentY=15.7981, extentZ=33.9157, center=<9.7551 0 6.0773>}]}
constraints []
