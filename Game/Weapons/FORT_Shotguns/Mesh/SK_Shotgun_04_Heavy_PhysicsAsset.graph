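# Physics asset for the SK_Shotgun_04_Heavy skeletal mesh, giving the heavy shotgun a simple collision shape for physics.
# A single body on the Handle bone holds one elongated box, offset and rotated from the bone origin, that stands in for the weapon.
# There are no constraints or other bodies, so the physics representation is one rigid piece with no articulated parts.
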PhysicsAsset /Game/Weapons/FORT_Shotguns/Mesh/SK_Shotgun_04_Heavy_PhysicsAsset
bodies {Handle [{shapeType=Box, extentX=11.0627, extentY=29.5779, extentZ=88.9094, center=<21.3556 -0.6891 6.7829>, rotation=<rot 0 89.586 -86.5862>}]}
constraints []
