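# Physics asset for the combat shotgun skeletal mesh, giving the weapon a simple collision shape.
# A single body on the Handle bone holds one elongated box, offset and rotated from the bone.
# It has no constraints, so the asset defines one rigid body with no jointed parts.
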PhysicsAsset /Game/Weapons/FORT_Shotguns/Mesh/SK_Shotgun_Combat_Physics
bodies {Handle [{shapeType=Box, extentX=9.0885, extentY=21.6853, extentZ=95.1767, center=<24.908 0.7195 2.3422>, rotation=<rot 0 90.0326 -88.3857>}]}
constraints []
