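# Physics asset for the retro sci-fi shotgun skeletal mesh, giving the weapon simple collision and a jointed ragdoll setup.
# Two capsule bodies cover the BreakAction and Handle bones.
# A single constraint links them, with matching angular limits on both swing axes and twist.
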PhysicsAsset /Game/Weapons/FORT_Shotguns/Mesh/SK_Shotgun_RetroSciFi_PhysicsAsset
bodies {
 BreakAction [{shapeType=Capsule, length=26.9181, radius=11.7488, center=<0.0483 15.1435 6.8319>, rotation=<rot 0 0.0154 -88.9508>}]
 Handle [{shapeType=Capsule, length=16.19, radius=11.6097, center=<0.4372 4.9805 -0.0259>, rotation=<rot 0 -179.9473 -75.7952>}]
}
constraints [{bone1Name=BreakAction, bone2Name=Handle, swing1LimitDegrees=45, swing2LimitDegrees=45, twistLimitDegrees=45}]
